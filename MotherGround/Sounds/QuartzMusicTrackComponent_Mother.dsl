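; Quartz music track component for Mother that mixes four track parameters (unaware, base, guard, boss) from data tracker awareness states of MotherWall and MotherGround.
; Each data tracker update calls the parent, then recomputes the mix from clamped 0–1 values.
; Unaware sums both unaware states, base sums all alerted and threatened states, and boss follows MotherGround threatened.
; Guard follows MotherWall threatened, but becomes the inverse of boss once boss reaches a small threshold.
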(event Custom|MIx%20State
  (Control|SetTrackParameter 0 (CallFunction|GetUnawareValue))
  (Control|SetTrackParameter 1 (CallFunction|GetBaseValue))
  (Control|SetTrackParameter 2 (Math|Float|SelectFloat (Math|Float|Lerp 1.0 0.0 (CallFunction|GetBossValue)) (CallFunction|GetGuardValue) (>= (CallFunction|GetBossValue) 0.01)))
  (Control|SetTrackParameter 3 (CallFunction|GetBossValue)))

(fn GetBaseValue
  (return (Math|Float|MapRangeClamped (+ (+ (CallFunction|GetDataTrackerParameterValue (Variables|Default|GetMotherWall_Threatened)) (CallFunction|GetDataTrackerParameterValue (Variables|Default|GetMotherGroundThreatened))) (+ (CallFunction|GetDataTrackerParameterValue (Variables|Default|GetMotherWall_Alerted)) (CallFunction|GetDataTrackerParameterValue (Variables|Default|GetMotherGround_Alerted)))) 0.0 1.0 0.0 1.0)))

(fn GetBossValue
  (return (Math|Float|MapRangeClamped (CallFunction|GetDataTrackerParameterValue (Variables|Default|GetMotherGroundThreatened)) 0.0 1.0 0.0 1.0)))

(fn GetGuardValue
  (return (Math|Float|MapRangeClamped (CallFunction|GetDataTrackerParameterValue (Variables|Default|GetMotherWall_Threatened)) 0.0 1.0 0.0 1.0)))

(fn GetUnawareValue
  (return (Math|Float|MapRangeClamped (+ (CallFunction|GetDataTrackerParameterValue (Variables|Default|GetMotherWall_Unaware)) (CallFunction|GetDataTrackerParameterValue (Variables|Default|GetMotherGround_Unaware))) 0.0 1.0 0.0 1.0)))

(fn OnDataTrackerUpdate
  (|CallParentFunction)
  (CallFunction|MIxState))
